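; Blueprint for the GlowDonut Light Post B environment prop, which exposes its static mesh parts for dynamic destruction.
; GetDynamicDestructionStaticMeshComponents returns three arrays: the structural parts (main mesh, up rod, light fixture), the urban light cone alone, and all four combined.
; Component variables come from Debug, AdjustableLight and Default categories, while the event graph and construction script hold no logic.
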(event EventBeginPlay)

(event Collision|EventActorBeginOverlap
  (params (OtherActor :type Actor%20Object%20Reference)))

(event EventTick
  (params (DeltaSeconds :type Float%20%28single-precision%29)))

(fn GetDynamicDestructionStaticMeshComponents
  (return (Utilities|Array|MakeArray (Variables|Debug|GetStaticMeshComponent) (Variables|AdjustableLight|GetUpRod) (Variables|AdjustableLight|GetLightFixture)) (Utilities|Array|MakeArray (Variables|Default|GetSM_Street_UrbanLightCone_01)) (Utilities|Array|MakeArray (Variables|Debug|GetStaticMeshComponent) (Variables|AdjustableLight|GetUpRod) (Variables|AdjustableLight|GetLightFixture) (Variables|Default|GetSM_Street_UrbanLightCone_01))))

(fn ConstructionScript)
